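# Animation Blueprint for the F_MED_BandageNinjaFNCS female player body, adding secondary motion to the pose copied from its attached parent mesh.
# A Control Rig adjusts that pose using smoothed crouch, skydive, float, emote and movement states from the event graph.
# AnimDynamics then simulates zipper and hood bones.
# An additive rotation tilts the left bag bone while jumping and not skydiving, despite its driving variable's name, "NOT Jumping and SkyDiving".
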AnimBlueprint /BRCosmetics/Characters/Player/Female/Medium/Bodies/F_MED_BandageNinjaFNCS/Meshes/F_MED_BandageNinjaFNCS_AnimBP
target_skeleton=&/F_MED_BandageNinjaFNCS_Skeleton
graphs [
 AnimGraph {nodes [
   n1 = Control Rig!(
    Source=Copy Pose From Mesh() {bUseAttachedParent=true}
    Alpha
    Is Crouching:double=Get Is Crouching() -> (Is Crouching:double)
    Is Skydiving:double=Get Is Sky Diving() -> (Is Sky Diving:double)
    Is Skydive Floating:double=Get Is Skydive Floating() -> (Is Skydive Floating:double)
    Is Playing Emote:double=Get Is Playing Emote() -> (Is Playing Emote:double)
    Is Moving 2D:double=Get is Moving 2D() -> (is Moving 2D:double)
   ) -> (Pose) {
    alphaBoolBlend {blendOption=Linear}
    controlRigAssetReference {blueprintRigClass=&/F_MED_BandageNinjaFNCS_CtrlRig.F_MED_BandageNinjaFNCS_CtrlRig_C}
   }
   n2 = Anim Dynamics - Bone: dyn_zipper(ComponentPose=Local To Component(LocalPose=@n1), Alpha=1) {
    angularDampingOverride=0.999
    angularSpringConstant=500
    bOverrideAngularDamping=true
    bUseGravityOverride=true
    gravityScale=-1
    simulationSpace=RootRelative
    alphaBoolBlend {blendOption=Linear}
    boundBone {boneName=dyn_zipper}
    componentAppliedLinearAccClamp=<5000 5000 5000>
    componentLinearAccScale=<0.2 0.2 0.2>
    componentLinearVelScale=<0.2 0.2 0.2>
    gravityOverride {z=-980}
    physicsBodyDefinitions [{
      sphereCollisionRadius=1
      boundBone {boneName=dyn_zipper}
      boxExtents=<10 5 5>
      constraintSetup {angularLimitsMax=<0 0 10>, angularLimitsMin=<0 -130 -10>, angularTarget=<0 0 0>}
      localJointOffset=<2 0 0>
     }]
   }
   n3 = Anim Dynamics - Bone: dyn_hood(ComponentPose=@n2, Alpha=1, ExternalForce="0,0,0") {
    angularDampingOverride=0.999
    angularSpringConstant=100
    bAngularSpring=true
    bLinearSpring=true
    bOverrideAngularDamping=true
    bOverrideLinearDamping=true
    bUseGravityOverride=true
    gravityScale=-1
    linearDampingOverride=0.999
    linearSpringConstant=150
    simulationSpace=RootRelative
    alphaBoolBlend {blendOption=Linear}
    boundBone {boneName=dyn_hood}
    componentAppliedLinearAccClamp=<2500 2500 2500>
    componentLinearAccScale=<0.6 0.6 0.6>
    componentLinearVelScale=<0.7 0.7 0.7>
    gravityOverride {z=-900}
    physicsBodyDefinitions [{
      boundBone {boneName=dyn_hood}
      boxExtents=<10 15 20>
      constraintSetup {angularLimitsMax=<0 20 10>, angularLimitsMin=<0 -20 -60>, angularTarget=<0 0 0>, linearAxesMax=<0 5 0>, linearAxesMin=<0 -10 0>}
      localJointOffset=<5 0 0>
     }]
   }
   n4 = "Transform (Modify) Bone - Bone: dyn_bag_01_l"!(
    ComponentPose=@n3
    bAlphaBoolEnabled:bool=Get NOT Jumping and SkyDiving() -> (NOT Jumping and SkyDiving:bool)
    Translation="0,0,0"
    Rotation="10,0,0"
    Scale="1,1,1"
   ) -> (Pose) {
    alphaInputType=Bool
    rotationMode=BMM_Additive
    rotationSpace=BCS_BoneSpace
    alphaBoolBlend {blendInTime=0.3, blendOutTime=1}
    boneToModify {boneName=dyn_bag_01_l}
    rotation {pitch=10}
   }
   Output Pose(Result=Component To Local(ComponentPose=@n4)) {layerGroup=None}
  ]}
 EventGraph {nodes [
   n1 = Event BlueprintUpdateAnimation()
   > Set Delta Time X(Delta Time X=@n1.DeltaTimeX)
   n2 = Get bIsCrouching()
   > Set Is Crouching(Is Crouching:double=FInterpTo(Current=Get Is Crouching(), Target="To Float (Boolean)"(InBool=@n2), DeltaTime=Get Delta Time X(), InterpSpeed=10)) -> (Output_Get:double)
   n3 = Get bIsSkyDiving()
   > Set Is Sky Diving(Is Sky Diving:double=FInterpTo(Current=Get Is Sky Diving(), Target="To Float (Boolean)"(InBool=@n3), DeltaTime=Get Delta Time X(), InterpSpeed=10)) -> (Output_Get:double)
   n4 = Get bIsSkydiveFloating()
   > Set Is Skydive Floating(Is Skydive Floating:double=FInterpTo(Current=Get Is Skydive Floating(), Target="To Float (Boolean)"(InBool=@n4), DeltaTime=Get Delta Time X(), InterpSpeed=10)) -> (Output_Get:double)
   n5 = Get bIsPlayingEmote()
   > Set Is Playing Emote(Is Playing Emote:double=FInterpTo(Current=Get Is Playing Emote(), Target="To Float (Boolean)"(InBool=@n5), DeltaTime=Get Delta Time X(), InterpSpeed=10)) -> (Output_Get:double)
   n6 = Get bIsMoving2D()
   > Set is Moving 2D(is Moving 2D:double=FInterpTo(Current=Get is Moving 2D(), Target="To Float (Boolean)"(InBool=@n6), DeltaTime=Get Delta Time X(), InterpSpeed=10)) -> (Output_Get:double)
   n7 = Get bIsSkyDiving()
   > Set NOT Jumping and SkyDiving(NOT Jumping and SkyDiving:bool=AND Boolean(A=Get bIsJumping(), B=NOT Boolean(A=@n7))) -> (Output_Get:bool)
  ]}
]
variables [
 Delta Time X {type=number}
 Is Crouching {type=number}
 is Moving 2D {type=number}
 Is Playing Emote {type=number}
 Is Sky Diving {type=number}
 Is Skydive Floating {type=number}
 NOT Jumping and SkyDiving {type=boolean}
]
