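# Anim Blueprint for the BandageNinjaFNCS female face accessory, adding secondary motion to its bangs and ponytail bones.
# It copies the attached parent's pose, then chains root-relative Anim Dynamics nodes with reduced gravity, heavy damping and per-bone angular limits.
# Only the second left bang has a spherical limit, driven by the head.
# Each update stores a deprecated gravity-override function's result in Gravity_Override, which no shown node reads.
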AnimBlueprint /BRCosmetics/Characters/Player/Female/Medium/Bodies/F_MED_BandageNinjaFNCS/Meshes/Parts/F_MED_BandageNinjaFNCS_FaceAcc_AnimBP
target_skeleton=&/F_MED_BandageNinjaFNCS_FaceAcc_Skeleton
graphs [
 AnimGraph {nodes [
   n1 = Anim Dynamics - Bone: dyn_bangs_01_l(ComponentPose=Local To Component(LocalPose=Copy Pose From Mesh() {bUseAttachedParent=true}), Alpha=1, ExternalForce="0,0,0") {
    angularDampingOverride=0.9999
    bGravityOverrideInSimSpace=true
    bOverrideAngularDamping=true
    gravityScale=0.35
    simulationSpace=RootRelative
    alphaBoolBlend {blendOption=Linear}
    boundBone {boneName=dyn_bangs_01_l}
    componentLinearAccScale {z=0.5}
    componentLinearVelScale {z=0.5}
    physicsBodyDefinitions [{
      sphereCollisionRadius=0
      boundBone {boneName=dyn_bangs_01_l}
      boxExtents=<5 20 20>
      constraintSetup {angularLimitsMax=<0 30 10>, angularLimitsMin=<0 0 -10>, angularTarget=<0 0 0>}
      localJointOffset=<5 0 0>
     }]
   }
   n2 = Anim Dynamics - Bone: dyn_bangs_02_l(ComponentPose=@n1, Alpha=1, ExternalForce="0,0,0") {
    angularDampingOverride=1
    bGravityOverrideInSimSpace=true
    bOverrideAngularDamping=true
    bUsePlanarLimit=false
    bUseSphericalLimits=true
    gravityScale=0.35
    simulationSpace=RootRelative
    alphaBoolBlend {blendOption=Linear}
    boundBone {boneName=dyn_bangs_02_l}
    componentLinearAccScale {z=0.5}
    componentLinearVelScale {z=0.5}
    physicsBodyDefinitions [{
      sphereCollisionRadius=0
      boundBone {boneName=dyn_bangs_02_l}
      boxExtents=<5 20 20>
      constraintSetup {angularLimitsMax=<0 30 10>, angularLimitsMin=<0 0 -10>, angularTarget=<0 0 0>}
      localJointOffset=<5 0 0>
     }]
    sphericalLimits [{limitRadius=9, drivingBone {boneName=head}, sphereLocalOffset=<5 0 0>}]
   }
   n3 = Anim Dynamics - Bone: dyn_bangs_01_r(ComponentPose=@n2, Alpha=1, ExternalForce="0,0,0") {
    angularDampingOverride=0.9999
    bGravityOverrideInSimSpace=true
    bOverrideAngularDamping=true
    gravityScale=0.35
    simulationSpace=RootRelative
    alphaBoolBlend {blendOption=Linear}
    boundBone {boneName=dyn_bangs_01_r}
    componentLinearAccScale {z=0.5}
    componentLinearVelScale {z=0.5}
    physicsBodyDefinitions [{
      sphereCollisionRadius=0
      boundBone {boneName=dyn_bangs_01_r}
      boxExtents=<5 20 20>
      constraintSetup {angularLimitsMax=<0 30 10>, angularLimitsMin=<0 -15 -10>, angularTarget=<0 0 0>}
      localJointOffset=<5 0 0>
     }]
   }
   n4 = Anim Dynamics - Bone: dyn_bangs_02_r(ComponentPose=@n3, Alpha=1, ExternalForce="0,0,0") {
    angularDampingOverride=1
    bGravityOverrideInSimSpace=true
    bOverrideAngularDamping=true
    gravityScale=0.35
    simulationSpace=RootRelative
    alphaBoolBlend {blendOption=Linear}
    boundBone {boneName=dyn_bangs_02_r}
    componentLinearAccScale {z=0.5}
    componentLinearVelScale {z=0.5}
    physicsBodyDefinitions [{
      sphereCollisionRadius=0
      boundBone {boneName=dyn_bangs_02_r}
      boxExtents=<5 20 20>
      constraintSetup {angularLimitsMax=<0 30 10>, angularLimitsMin=<0 0 -10>, angularTarget=<0 0 0>}
      localJointOffset=<5 0 0>
     }]
   }
   n5 = Anim Dynamics - Bone: dyn_ponytail_01(ComponentPose=@n4, Alpha=1, ExternalForce="0,0,0") {
    angularDampingOverride=0.9999
    bGravityOverrideInSimSpace=true
    bOverrideAngularDamping=true
    gravityScale=0.35
    simulationSpace=RootRelative
    alphaBoolBlend {blendOption=Linear}
    boundBone {boneName=dyn_ponytail_01}
    componentAppliedLinearAccClamp=<2000 2000 2000>
    componentLinearAccScale {y=0.5, z=0.5}
    componentLinearVelScale {y=0.5, z=0.5}
    physicsBodyDefinitions [{
      sphereCollisionRadius=0
      boundBone {boneName=dyn_ponytail_01}
      boxExtents=<7 7 7>
      constraintSetup {angularLimitsMax=<0 5 10>, angularLimitsMin=<0 -10 -10>, angularTarget=<0 0 0>}
      localJointOffset=<0.4 0 0>
     }]
   }
   n6 = Anim Dynamics - Bone: dyn_ponytail_02(ComponentPose=@n5, Alpha=1, ExternalForce="0,0,0") {
    angularDampingOverride=0.9999
    bGravityOverrideInSimSpace=true
    bOverrideAngularDamping=true
    gravityScale=0.3
    simulationSpace=RootRelative
    alphaBoolBlend {blendOption=Linear}
    boundBone {boneName=dyn_ponytail_02}
    componentLinearAccScale {y=0.5, z=0.5}
    componentLinearVelScale {y=0.5, z=0.5}
    physicsBodyDefinitions [{
      sphereCollisionRadius=0
      boundBone {boneName=dyn_ponytail_02}
      boxExtents=<7 5 5>
      constraintSetup {angularLimitsMax=<0 5 5>, angularLimitsMin=<0 -10 -5>, angularTarget=<0 0 0>}
      localJointOffset=<0.2 0 0>
     }]
   }
   n7 = Anim Dynamics - Bone: dyn_ponytail_03(ComponentPose=@n6, Alpha=1, ExternalForce="0,0,0") {
    angularDampingOverride=0.9999
    bGravityOverrideInSimSpace=true
    bOverrideAngularDamping=true
    gravityScale=0.35
    simulationSpace=RootRelative
    alphaBoolBlend {blendOption=Linear}
    boundBone {boneName=dyn_ponytail_03}
    componentAppliedLinearAccClamp=<2500 2500 2500>
    componentLinearAccScale {y=0.5, z=0.5}
    componentLinearVelScale {y=0.5, z=0.5}
    physicsBodyDefinitions [{
      sphereCollisionRadius=0
      boundBone {boneName=dyn_ponytail_03}
      boxExtents=<6 6 6>
      constraintSetup {angularLimitsMax=<0 5 10>, angularLimitsMin=<0 -40 -10>, angularTarget=<0 0 0>}
      localJointOffset=<0.1 0 0>
     }]
   }
   n8 = Anim Dynamics - Bone: dyn_ponytail_04(ComponentPose=@n7, Alpha=1, ExternalForce="0,0,0") {
    angularDampingOverride=0.9999
    bGravityOverrideInSimSpace=true
    bOverrideAngularDamping=true
    gravityScale=0.35
    simulationSpace=RootRelative
    alphaBoolBlend {blendOption=Linear}
    boundBone {boneName=dyn_ponytail_04}
    componentAppliedLinearAccClamp=<2500 2500 2500>
    componentLinearAccScale {y=0.5, z=0.5}
    componentLinearVelScale {y=0.5, z=0.5}
    physicsBodyDefinitions [{
      sphereCollisionRadius=0
      boundBone {boneName=dyn_ponytail_04}
      boxExtents=<8 5 5>
      constraintSetup {angularLimitsMax=<0 0 10>, angularLimitsMin=<0 -40 -10>, angularTarget=<0 0 0>}
      localJointOffset=<0.3 0 0>
     }]
   }
   n9 = Anim Dynamics - Bone: dyn_ponytail_05(ComponentPose=@n8, Alpha=1, ExternalForce="0,0,0") {
    angularDampingOverride=0.9999
    bGravityOverrideInSimSpace=true
    bOverrideAngularDamping=true
    gravityScale=0.35
    simulationSpace=RootRelative
    alphaBoolBlend {blendOption=Linear}
    boundBone {boneName=dyn_ponytail_05}
    componentAppliedLinearAccClamp=<1500 1500 1500>
    componentLinearAccScale {y=0.5, z=0.5}
    componentLinearVelScale {y=0.5, z=0.5}
    physicsBodyDefinitions [{
      sphereCollisionRadius=0
      boundBone {boneName=dyn_ponytail_05}
      boxExtents=<10 5 5>
      constraintSetup {angularLimitsMax=<0 5 10>, angularLimitsMin=<0 -20 -10>, angularTarget=<0 0 0>}
      localJointOffset=<0.3 0 0>
     }]
   }
   n10 = Anim Dynamics - Bone: dyn_ponytail_06(ComponentPose=@n9, Alpha=1, ExternalForce="0,0,0") {
    angularDampingOverride=0.9999
    bGravityOverrideInSimSpace=true
    bOverrideAngularDamping=true
    gravityScale=0.4
    simulationSpace=RootRelative
    alphaBoolBlend {blendOption=Linear}
    boundBone {boneName=dyn_ponytail_06}
    componentAppliedLinearAccClamp=<1500 1500 1500>
    componentLinearAccScale {y=0.5, z=0.5}
    componentLinearVelScale {y=0.5, z=0.5}
    physicsBodyDefinitions [{
      sphereCollisionRadius=0
      boundBone {boneName=dyn_ponytail_06}
      boxExtents=<9 5 5>
      constraintSetup {angularLimitsMax=<0 5 15>, angularLimitsMin=<0 -50 -15>, angularTarget=<0 0 0>}
      localJointOffset=<0.7 0 0>
     }]
   }
   n11 = Anim Dynamics - Bone: dyn_ponytail_07(ComponentPose=@n10, Alpha=1, ExternalForce="0,0,0") {
    angularDampingOverride=0.9999
    bGravityOverrideInSimSpace=true
    bOverrideAngularDamping=true
    gravityScale=0.35
    simulationSpace=RootRelative
    alphaBoolBlend {blendOption=Linear}
    boundBone {boneName=dyn_ponytail_07}
    componentLinearAccScale {y=0.5, z=0.5}
    componentLinearVelScale {y=0.5, z=0.5}
    physicsBodyDefinitions [{
      sphereCollisionRadius=0
      boundBone {boneName=dyn_ponytail_07}
      boxExtents=<10 5 5>
      constraintSetup {angularLimitsMax=<0 5 15>, angularLimitsMin=<0 -40 -15>, angularTarget=<0 0 0>}
      localJointOffset=<0.2 0 0>
     }]
   }
   Output Pose(Result=Component To Local(ComponentPose=@n11)) {layerGroup=None}
  ]}
 EventGraph {nodes [
   Event BlueprintUpdateAnimation()
   > n1 = Gravity Override State Function DEPRECATED(CharacterPartAnimInstance=Self-Reference(), GameTimeInSeconds=GetGameTimeInSeconds(), GravityOverrideParams=Get Gravity Override Params())
   > Set Gravity_Override(Gravity_Override:Vector=@n1) -> (Output_Get:Vector)
  ]}
]
variables [
 Gravity Override Params {type=GravityOverrideParamsStruct}
 Gravity_Override {type=Vector}
]
